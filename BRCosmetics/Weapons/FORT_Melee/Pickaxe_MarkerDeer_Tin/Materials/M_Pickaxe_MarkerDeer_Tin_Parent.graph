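# Parent material for the MarkerDeer Tin pickaxe that adds a glow driven by hit count to the shared 2023 character parent functions.
# Blueprint-fed time and hit parameters compute a progress value that eases per hit, then fades to zero after a delayed reset.
# A smoothstep on pre-skinned local position masks the two-color Fresnel emissive, with a sphere mask flagging peak hits.
# Overrides of base color, roughness, specular and emissive sit between the PreFX and PostFX functions.
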Material /BRCosmetics/Weapons/FORT_Melee/Pickaxe_MarkerDeer_Tin/Materials/M_Pickaxe_MarkerDeer_Tin_Parent
graph {
 properties {
  bUseMaterialAttributes=true
  bUsedWithMeshDeformer=true
  bUsedWithMorphTargets=true
  bUsedWithSkeletalMesh=true
  pixelDepthOffsetMode=PDOM_Legacy
  previewMesh=/BRCosmetics/Weapons/FORT_Melee/Pickaxe_MarkerDeer_Tin/Meshes/MarkerDeer_Tin_Axe
  shadingModel=MSM_FromMaterialExpression
  usedShadingModels=Default Lit | Subsurface | Clear Coat | Subsurface Profile | Cloth
 }
 nodes [
  n1 = MaterialFunctionCall() {materialFunction=&MF_CharacterParent_2023_PreFX}
  n2 = GetMaterialAttributes(None=@n1.Result) {attributeGetTypes [
    #1
    #2
    #3
    #4
   ]}
  n3 = SetMaterialAttributes!(MaterialAttributes=@n2.MaterialAttributes, Emissive Color=NamedRerouteUsage(), Base Color=NamedRerouteUsage(), Roughness, Specular) {attributeSetTypes [
    #1
    #2
    #3
    #4
   ]}
  MaterialFunctionCall(Material=@n3, M=@n1.M, Pre Skinned Local Position=@n1.M, Distance Blend=@n1.M, Fresnel=@n1.M) {materialFunction=&MF_CharacterParent_2023_PostFX}
  NamedRerouteDeclaration(None=@n2.Emissive Color) {name=Emissive_IN, nodeColor=<rgba 0 0.9531 1 1>}
  n4 = ComponentMask(None=VertexInterpolator(VS=LocalPosition() {localOrigin=InstancePreSkinning}.XYZ).PS) {b=true}
  n5 = Multiply(
   A=NamedRerouteDeclaration(None=SmoothStep(Value=@n4) {constMax=-80, constMin=-65}) {name=GlowMask, nodeColor=<rgba 1 0 0.5625 1>}
   B=ScalarParameter() {defaultValue=3, group=MarkerDeerTin, parameterName=MDT_GlowIntensity}
  )
  n6 = VectorParameter() {group=MarkerDeerTin, parameterName=MDT_GlowColor, defaultValue=<rgba 0.6654 0 0.0145 1>}
  n7 = Fresnel() {baseReflectFraction=0, exponent=1}
  n8 = ScalarParameter() {defaultValue=0.1, parameterName=Power}
  n9 = LinearInterpolate(
   A=@n6.RGB
   B=LinearInterpolate(
    A=VectorParameter() {group=MarkerDeerTin, parameterName=MDT_GlowColor2, defaultValue=<rgba 2 0.7694 0.3646 1>}.RGB
    B=@n6.RGB
    Alpha=Power(Base=@n7, Exp=@n8) {constExponent=0.3}
   )
   Alpha=NamedRerouteUsage()
  )
  NamedRerouteDeclaration(None=Multiply(A=Multiply(A=@n5, B=@n9), B=NamedRerouteUsage())) {name=Emissive_OUT, nodeColor=<rgba 1 0.7266 0 1>}
  n10 = NamedRerouteUsage()
  n11 = NamedRerouteUsage()
  NamedRerouteDeclaration(None=LinearInterpolate(A=@n10, B=LinearInterpolate(A=@n10, Alpha=@n11) {constB=0}, Alpha=NamedRerouteUsage())) {name=Specular_OUT, nodeColor=<rgba 1 0 0.4219 1>}
  n12 = NamedRerouteUsage()
  n13 = NamedRerouteUsage()
  n14 = OneMinus(None=@n13)
  n15 = NamedRerouteUsage()
  NamedRerouteDeclaration(None=LinearInterpolate(A=@n12, B=LinearInterpolate(A=@n12, B=@n14, Alpha=@n13), Alpha=Power(Base=@n15) {constExponent=0.5})) {name=Basecolor_OUT, nodeColor=<rgba 1 0.2344 0 1>}
  NamedRerouteDeclaration(None=@n2.Base Color) {name=Basecolor_IN, nodeColor=<rgba 0 0.0859 1 1>}
  n16 = ScalarParameter() {group=MarkerDeerTin, parameterName=MDT_CurrHitCount}
  n17 = ScalarParameter() {defaultValue=7, group=MarkerDeerTin, parameterName=MDT_ResetHits}
  n18 = Divide(A=Saturate(None=Subtract(A=@n16)), B=@n17)
  n19 = Divide(A=@n16, B=@n17)
  n20 = Time()
  n21 = ScalarParameter() {group=MarkerDeerTin, parameterName=MDT_GameTimeFromBP}
  n22 = Saturate(None=Divide(A=Subtract(A=@n20, B=@n21), B=ScalarParameter() {defaultValue=1, group=MarkerDeerTin, parameterName=MDT_GlowInterval}))
  n23 = Time()
  n24 = ScalarParameter() {group=MarkerDeerTin, parameterName=MDT_GameTimeFromBP_Reset}
  n25 = Add(A=@n24, B=ScalarParameter() {defaultValue=2, group=MarkerDeerTin, parameterName=MDT_ResetDelay})
  n26 = Saturate(None=Divide(A=Subtract(A=@n23, B=@n25), B=ScalarParameter() {defaultValue=2, group=MarkerDeerTin, parameterName=MDT_ResetDuration}))
  n27 = Saturate(None=Sign(None=@n24))
  NamedRerouteDeclaration(None=LinearInterpolate(A=LinearInterpolate(A=@n18, B=@n19, Alpha=@n22), Alpha=Multiply(A=@n26, B=@n27)) {constB=0}) {name=Progress, nodeColor=<rgba 1 0 0.8203 1>}
  NamedRerouteDeclaration(None=@n2.Roughness) {name=Roughness_IN, nodeColor=<rgba 0 1 0.8594 1>}
  n28 = NamedRerouteUsage()
  n29 = NamedRerouteUsage()
  NamedRerouteDeclaration(None=LinearInterpolate(A=@n28, B=LinearInterpolate(A=@n28, Alpha=@n29), Alpha=NamedRerouteUsage())) {name=Roughness_OUT, nodeColor=<rgba 1 0.5391 0 1>}
  NamedRerouteDeclaration(None=@n2.Specular) {name=Specular_IN, nodeColor=<rgba 0.2188 0 1 1>}
  NamedRerouteDeclaration(None=SphereMask(A=@n16, B=@n17) {attenuationRadius=1, hardnessPercent=0.5}) {name=IsPeak, nodeColor=<rgba 0.5703 0 1 1>}
  NamedRerouteUsage()
  NamedRerouteUsage()
  n30 = MaterialFunctionCall(Float3=VertexInterpolator(VS=PreSkinnedNormal()).PS) {materialFunction=&BreakOutFloat3Components}
  Normalize(VectorInput=MaterialFunctionCall(X=@n30.R, Y=@n30.R, Z=Constant()) {materialFunction=&MakeFloat3}.Result)
  ScalarParameter() {group=MarkerDeerTin, parameterName=MDT_GlowProgress}
  n31 = LocalPosition() {localOrigin=InstancePreSkinning}
  n32 = VectorParameter() {group=Toon_Shading, parameterName=NormalSphere_Location, defaultValue=<rgba 0 8.2454 0 1>}
  n33 = Normalize(VectorInput=Multiply(A=Subtract(A=@n31.XYZ, B=@n32.RGB), B=Constant3Vector() {constant=<rgba 1 1 0 0>}))
  ShaderStageSwitch(PixelShader=VertexInterpolator(VS=@n33).PS, VertexShader=@n33)
  SmoothStep(Value=@n4) {constMax=-110, constMin=-70}
  VertexNormalWS()
 ]
}
