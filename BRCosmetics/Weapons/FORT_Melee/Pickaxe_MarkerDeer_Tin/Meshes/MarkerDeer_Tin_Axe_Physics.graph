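# Physics asset for the MarkerDeer Tin pickaxe cosmetic, giving its three dangling wire chains physical bodies and joints.
# The two-handed handle carries four capsules, and each chain runs from a base bone through four dyn_wire bones, one capsule apiece.
# Angular motion is free from handle to base and base to first segment, locked to the second, then swing-limited toward the tip, with twist tightly limited at the last joint.
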PhysicsAsset /BRCosmetics/Weapons/FORT_Melee/Pickaxe_MarkerDeer_Tin/Meshes/MarkerDeer_Tin_Axe_Physics
bodies {
 Melee_TwoHanded_Handle [
  {shapeType=Capsule, length=29.5, radius=8, center=<6.84 24.6671 116.4055>, rotation=<rot 0 0 74.9996>}
  {shapeType=Capsule, length=29.5, radius=20, center=<-15.2455 31.7317 96.045>, rotation=<rot 0 0 179.9996>}
  {shapeType=Capsule, length=29.5, radius=20, center=<29.3531 31.5529 96.045>, rotation=<rot 0 0 179.9996>}
  {shapeType=Capsule, length=29.5, radius=7, center=<6.84 24.6673 70.8621>, rotation=<rot 0 0 89.9998>}
 ]
 base_wire_a [{shapeType=Capsule, length=1, radius=1, center=<1 0 0>, rotation=<rot -90 0 0>}]
 base_wire_b [{shapeType=Capsule, length=1, radius=1, center=<1 0 0>, rotation=<rot -90 0 0>}]
 base_wire_c [{shapeType=Capsule, length=1, radius=1, center=<1 0 0>, rotation=<rot -90 0 0>}]
 dyn_wire_a1 [{shapeType=Capsule, length=1, radius=1, center=<1 0 0>, rotation=<rot -90 0 0>}]
 dyn_wire_a2 [{shapeType=Capsule, length=6, radius=2, center=<2 0 0>, rotation=<rot -90 0 0>}]
 dyn_wire_a3 [{shapeType=Capsule, length=4, radius=2, center=<1 0 0>, rotation=<rot -90 0 0>}]
 dyn_wire_a4 [{shapeType=Capsule, length=5, radius=2, center=<6 0 0>, rotation=<rot -90 0 0>}]
 dyn_wire_b1 [{shapeType=Capsule, length=1, radius=1, center=<1 0 0>, rotation=<rot -90 0 0>}]
 dyn_wire_b2 [{shapeType=Capsule, length=6, radius=2, center=<2 0 -1>, rotation=<rot -90 180 180>}]
 dyn_wire_b3 [{shapeType=Capsule, length=4, radius=2, center=<1 0 0>, rotation=<rot -90 0 0>}]
 dyn_wire_b4 [{shapeType=Capsule, length=5, radius=2, center=<6 0 0>, rotation=<rot -90 0 0>}]
 dyn_wire_c1 [{shapeType=Capsule, length=1, radius=1, center=<1 0 0>, rotation=<rot -90 0 0>}]
 dyn_wire_c2 [{shapeType=Capsule, length=12, radius=2, center=<2 0 0>, rotation=<rot -90 0 0>}]
 dyn_wire_c3 [{shapeType=Capsule, length=4, radius=1.5, center=<1 0 0>, rotation=<rot -90 0 0>}]
 dyn_wire_c4 [{shapeType=Capsule, length=5, radius=1.5, center=<6 0 0>, rotation=<rot -90 0 0>}]
}
constraints [
 {
  bone1Name=dyn_wire_a1
  bone2Name=base_wire_a
  swing1LimitDegrees=45
  swing1Motion=Free
  swing2LimitDegrees=45
  swing2Motion=Free
  twistLimitDegrees=45
  twistMotion=Free
 }
 {
  bone1Name=dyn_wire_a2
  bone2Name=dyn_wire_a1
  swing1LimitDegrees=45
  swing1Motion=Locked
  swing2LimitDegrees=25
  swing2Motion=Locked
  twistLimitDegrees=45
  twistMotion=Locked
 }
 {bone1Name=dyn_wire_a3, bone2Name=dyn_wire_a2, swing1LimitDegrees=20, swing2LimitDegrees=10, twistLimitDegrees=45, twistMotion=Locked}
 {
  bone1Name=dyn_wire_b1
  bone2Name=base_wire_b
  swing1LimitDegrees=45
  swing1Motion=Free
  swing2LimitDegrees=45
  swing2Motion=Free
  twistLimitDegrees=45
  twistMotion=Free
 }
 {
  bone1Name=dyn_wire_b2
  bone2Name=dyn_wire_b1
  swing1LimitDegrees=45
  swing1Motion=Locked
  swing2LimitDegrees=30
  swing2Motion=Locked
  twistLimitDegrees=45
  twistMotion=Locked
 }
 {bone1Name=dyn_wire_b3, bone2Name=dyn_wire_b2, swing1LimitDegrees=20, swing2LimitDegrees=10, twistLimitDegrees=45, twistMotion=Locked}
 {
  bone1Name=dyn_wire_c1
  bone2Name=base_wire_c
  swing1LimitDegrees=45
  swing1Motion=Free
  swing2LimitDegrees=45
  swing2Motion=Free
  twistLimitDegrees=45
  twistMotion=Free
 }
 {
  bone1Name=dyn_wire_c2
  bone2Name=dyn_wire_c1
  swing1LimitDegrees=45
  swing1Motion=Locked
  swing2LimitDegrees=25
  swing2Motion=Locked
  twistLimitDegrees=45
  twistMotion=Locked
 }
 {bone1Name=dyn_wire_c3, bone2Name=dyn_wire_c2, swing1LimitDegrees=20, swing2LimitDegrees=10, twistLimitDegrees=45, twistMotion=Locked}
 {bone1Name=dyn_wire_c4, bone2Name=dyn_wire_c3, swing1LimitDegrees=20, swing2LimitDegrees=10, twistLimitDegrees=1}
 {bone1Name=dyn_wire_b4, bone2Name=dyn_wire_b3, swing1LimitDegrees=20, swing2LimitDegrees=10, twistLimitDegrees=1}
 {bone1Name=dyn_wire_a4, bone2Name=dyn_wire_a3, swing1LimitDegrees=20, swing2LimitDegrees=10, twistLimitDegrees=1}
 {
  bone1Name=base_wire_b
  bone2Name=Melee_TwoHanded_Handle
  swing1LimitDegrees=45
  swing1Motion=Free
  swing2LimitDegrees=45
  swing2Motion=Free
  twistLimitDegrees=45
  twistMotion=Free
 }
 {
  bone1Name=base_wire_c
  bone2Name=Melee_TwoHanded_Handle
  swing1LimitDegrees=45
  swing1Motion=Free
  swing2LimitDegrees=45
  swing2Motion=Free
  twistLimitDegrees=45
  twistMotion=Free
 }
 {
  bone1Name=base_wire_a
  bone2Name=Melee_TwoHanded_Handle
  swing1LimitDegrees=45
  swing1Motion=Free
  swing2LimitDegrees=45
  swing2Motion=Free
  twistLimitDegrees=45
  twistMotion=Free
 }
]
